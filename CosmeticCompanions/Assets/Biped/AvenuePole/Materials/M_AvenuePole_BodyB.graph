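# Body material for the AvenuePole biped cosmetic companion, built on material attributes for skeletal meshes with morph targets.
# The character parent PreFX function feeds companion custom color and an AvenuePole flipbook base, then the parent PostFX pass.
# World position offset swaps mesh and flipbook mouths: a thresholded rerouted value drives two opposite vertex-color blue masks that collapse hidden vertices to the object position.
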Material /CosmeticCompanions/Assets/Biped/AvenuePole/Materials/M_AvenuePole_BodyB
graph {
 properties {
  bAutomaticallySetUsageInEditor=false
  bUseMaterialAttributes=true
  bUsedWithMeshDeformer=true
  bUsedWithMorphTargets=true
  bUsedWithSkeletalMesh=true
  bUsedWithStaticLighting=true
  previewMesh=/CosmeticCompanions/Assets/Biped/AvenuePole/Meshes/AvenuePole_Biped
  shadingModel=MSM_FromMaterialExpression
  usedShadingModels=Default Lit | Subsurface | Clear Coat | Subsurface Profile | Cloth
 }
 nodes [
  Constant() {r=1}
  n1 = Constant3Vector()
  If(A=NamedRerouteUsage(), "A > B"=@n1, "A == B"=@n1, "A < B"=NamedRerouteUsage()) {constB=0.99}
  n2 = MaterialFunctionCall() {materialFunction=&MF_CharacterParent_2023_PreFX}
  n3 = MaterialFunctionCall!(In=MaterialFunctionCall!(Material=@n2.Result, Tech Art Mask=@n2.Result) {materialFunction=/CosmeticCompanions/Materials/MaterialFunctions/MF_FN_Companion_Custom_Color}.Base Color) {materialFunction=&/MF_AvenuePole_SK_Flipbook_Base}
  NamedRerouteDeclaration(None=@n3.MouthCurveIndex) {name=Out_MouthCurveIndex, nodeColor=<rgba 0.4531 1 0 1>}
  n4 = GetMaterialAttributes(None=MaterialFunctionCall(Material=@n3.Result, M=@n2.M, Pre Skinned Local Position=@n2.M, Distance Blend=@n2.M, Fresnel=@n2.M) {materialFunction=&MF_CharacterParent_2023_PostFX}.Result) {attributeGetTypes [#1]}
  NamedRerouteDeclaration(None=@n4.World Position Offset) {name=WPO_Out, nodeColor=<rgba 1 0 0.7734 1>}
  NamedRerouteUsage()
  n5 = ObjectPositionWS()
  n6 = WorldPosition()
  n7 = If(A=NamedRerouteUsage(), "A > B"=Constant(), "A < B"=Constant() {r=1}) {constB=0.5}
  n8 = VertexColor()
  n9 = MaterialFunctionCall!(Hide Element Toggle=@n7, VertexColorChannel=@n8.B, Vertex Value to hide=Constant() {r=0.1}) {desc="1 = mesh mouth, 0 = flipbook mouth", materialFunction=/Game/Packages/Fortress_SharedMaterials/Base_Material_Functions/VertexColorMask}
  n10 = MaterialFunctionCall(Hide Element Toggle=OneMinus(None=@n7), VertexColorChannel=@n8.B, Vertex Value to hide=Constant() {r=0.2}) {desc="0 = mesh mouth, 1 = flipbook mouth", materialFunction=&VertexColorMask}
  SetMaterialAttributes!(MaterialAttributes=@n4.MaterialAttributes, World Position Offset=Multiply(A=Subtract(A=@n5, B=@n6.XYZ), B=Max(A=@n9.Result, B=@n10.Result))) {attributeSetTypes [#1]}
 ]
}
