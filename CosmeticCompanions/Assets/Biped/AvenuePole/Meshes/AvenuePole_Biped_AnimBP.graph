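# Animation Blueprint for the AvenuePole biped cosmetic companion that drives its skeleton by retargeting the pose from another mesh.
# Blend Poses by bool selects the interaction IK Retargeter while isInteraction is true, and the standard one otherwise.
# Each update sets that flag by testing whether the isInteracting curve is above zero.
# A Copy Pose From Mesh node and a TryGetPawnOwner call are left unconnected.
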AnimBlueprint /CosmeticCompanions/Assets/Biped/AvenuePole/Meshes/AvenuePole_Biped_AnimBP
target_skeleton=&/AvenuePole_Biped_Skeleton
graphs [
 AnimGraph {nodes [
   Copy Pose From Mesh() {bCopyCurves=true, bUseAttachedParent=true}
   Output Pose(Result=Blend Poses by bool(
     BlendPose_0=Retarget Pose From Mesh!() -> (Pose:Pose Link Structure) {
      is=title AnimGraphNode_RetargetPoseFromMesh
      iKRetargeterAsset=&/RTG_AvenuePole_Biped_Interact
     }
     BlendPose_1=Retarget Pose From Mesh() {iKRetargeterAsset=&/RTG_AvenuePole_Biped}
     bActiveValue=Get isInteraction() -> (isInteraction:bool)
    ))
  ]}
 EventGraph {nodes [
   Event BlueprintUpdateAnimation()
   > n1 = Set isInteraction_float(isInteraction_float:double=GetCurveValue(CurveName=isInteracting)) -> (Output_Get:double)
   > Set isInteraction(isInteraction:bool="float > float"(A=@n1)) -> (Output_Get:bool)
   TryGetPawnOwner()
  ]}
]
variables [
 isInteraction {type=boolean}
 isInteraction_float {type=number}
]
